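# Conversation database for the EvokeFind NPC, a medic specialist who speaks as the maknae of HUNTR/X.
# Its entry point opens a greeting Speech node with contextual lines for a first-ever conversation, a conversation-count check and specific player outfits, using priority curves for first-meeting and outfit reactions.
# The greeting then branches through link nodes to the shared quest and basic-service entries and to EvokeFind's own dialogue entry.
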ConversationDatabase /NPCLibrary/NPCs/EvokeFind/NPCLibrary_Conversation_EvokeFind
entry_points [{tag=Conversation.Entry.EvokeFind, destinations [#1]}]
nodes [
 ConversationEntryPointNode_7 {
  id=#1
  class=ConversationEntryPointNode
  outputs [#2]
  properties {entryTag {tagName=Conversation.Entry.EvokeFind}}
 }
 Speech_C_4 {
  id=#2
  class=Speech_C
  outputs [#3, #4, #5]
  properties {
   nodeName=""
   generalConfig {
    defaultMessage="If you're looking for the maknae of HUNTR/X, you got her!"
    contextualMessages [
     {
      message="We heard your Island has a little demon problem!"
      contextRequirements [{requirement {class=FortControllerRequirement_IsFirstEverConversationWithNPC}}]
      priority {curve {rowName=Tandem.Speech.Priority.FirstEverInteraction}}
     }
     {
      message="If you're looking for the maknae of HUNTR/X, you got her!"
      contextRequirements [{requirement {
         class=FortControllerRequirement_CheckNPCMatchConversationCount
         properties {nPCUniqueIdTag {tagName=AISpawnerData.Type.Specialist.Medic.EvokeFind}}
        }}]
      priority {curve {rowName=None}}
     }
     {
      message="Couch! Couch! Couch! Couch!"
      contextRequirements [{requirement {class=FortControllerRequirement_HasCID, properties {allowedCIds [/BRCosmetics/Athena/Items/Cosmetics/Characters/Character_QuailWink]}}}]
      priority {curve {rowName=Tandem.Speech.Priority.OutfitReaction}}
     }
     {
      message="So, what's the plan?"
      contextRequirements [{requirement {class=FortControllerRequirement_HasCID, properties {allowedCIds [/BRCosmetics/Athena/Items/Cosmetics/Characters/Character_LoanFloat]}}}]
      priority {curve {rowName=Tandem.Speech.Priority.OutfitReaction}}
     }
     {
      message=Demon? Ohh... You'd better start running.
      contextRequirements [{requirement {
         class=FortControllerRequirement_HasCID
         properties {allowedCIds [
           /BRCosmetics/Athena/Items/Cosmetics/Characters/Character_UnluckyRoll
           /BRCosmetics/Athena/Items/Cosmetics/Characters/CID_A_466_Athena_Commando_F_Chaos
           /BRCosmetics/Athena/Items/Cosmetics/Characters/Character_ChaosDarkIce
           /BRCosmetics/Athena/Items/Cosmetics/Characters/Character_PocketScrunchie
           /BRCosmetics/Athena/Items/Cosmetics/Characters/Character_PointyTemper
           /BRCosmetics/Athena/Items/Cosmetics/Characters/Character_LilacLeather
           &Character_RoseDepth_Seed
          ]}
        }}]
      priority {curve {rowName=Tandem.Speech.Priority.OutfitReaction}}
     }
    ]
   }
  }
 }
 ConversationLinkNode_2 {id=#5, class=ConversationLinkNode, properties {remoteEntryTag {tagName=Conversation.Entry.EvokeFind.Dialogue}}}
 ConversationLinkNode_0 {id=#4, class=ConversationLinkNode, properties {remoteEntryTag {tagName=Conversation.Entry.BasicServices}}}
 ConversationLinkNode_1 {id=#3, class=ConversationLinkNode, properties {remoteEntryTag {tagName=Conversation.Entry.NPCLibrary.Quests}}}
]
